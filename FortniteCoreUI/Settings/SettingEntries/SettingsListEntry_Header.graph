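# Section header row for the settings list, a Widget Blueprint derived from FortSettingsListEntry_Setting.
# An invisible, unpadded Border wraps the inherited Text_SettingName, styled as uppercase bold italic HeadingNow text with a scroll style and slightly larger mobile font.
# The inherited Background widget is left unassigned.
# The event graph is empty, so the header text placeholder and any behavior rely on the parent class.
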WidgetBlueprint /FortniteCoreUI/Settings/SettingEntries/SettingsListEntry_Header
graphs [EventGraph {links [], nodes []}]
variables []
widgets {
 info {inheritedWidgetCount=2, parentClass=/Script/FortniteUI.FortSettingsListEntry_Setting, rootWidgetClass=&Border, widgetCount=3}
 widgets [
  Border_0 {
   widgetClassPath=&Border
   widget=Border_0
   properties {
    background {
     bIsDynamicallyLoaded=false
     drawAs=NoDrawType
     imageType=NoImage
     mirroring=NoMirror
     resourceName=None
     resourceObject=null
     tiling=NoTile
     imageSize=<32 32>
     margin=<ltrb 0 0 0 0>
     outlineSettings {
      bUseBrushTransparency=false
      roundingType=HalfHeightRadius
      width=0
      color {colorUseRule=UseColor_Specified, specifiedColor=<rgba 0 0 0 0>}
      cornerRadii=<0 0 0 0>
     }
     tintColor {colorUseRule=UseColor_Specified, specifiedColor=<rgba 1 1 1 1>}
     uVRegion {bIsValid=false, max=<0 0>, min=<0 0>}
    }
    padding=<ltrb 0 0 0 0>
   }
  }
  Text_SettingName {
   widgetClassPath=&CommonTextBlock
   bInherited=true
   bIsVariable=true
   parent=Border_0
   widget=Text_SettingName
   properties {
    mobileFontSizeMultiplier=1.1
    scrollStyle=&TextScrollStyle-Base_C
    text="{Header Text}"
    textTransformPolicy=ToUpper
    font {
     bForceMonospaced=false
     bMaterialIsStencil=false
     fontMaterial=null
     fontObject=&HeadingNow
     letterSpacing=0
     monospacedWidth=1
     size=25.5
     skewAmount=0
     typefaceFontName=76 Bold Italic
     outlineSettings {
      bApplyOutlineToDropShadows=false
      bMiteredCorners=false
      bSeparateFillAlpha=false
      outlineMaterial=null
      outlineSize=0
      outlineColor=<rgba 0 0 0 1>
     }
    }
    shadowOffset=<0 0>
    strikeBrush {
     bIsDynamicallyLoaded=false
     drawAs=NoDrawType
     imageType=NoImage
     mirroring=NoMirror
     resourceName=None
     resourceObject=null
     tiling=NoTile
     imageSize=<32 32>
     margin=<ltrb 0 0 0 0>
     outlineSettings {
      bUseBrushTransparency=false
      roundingType=HalfHeightRadius
      width=0
      color {colorUseRule=UseColor_Specified, specifiedColor=<rgba 0 0 0 0>}
      cornerRadii=<0 0 0 0>
     }
     tintColor {colorUseRule=UseColor_Specified, specifiedColor=<rgba 1 1 1 1>}
     uVRegion {bIsValid=false, max=<0 0>, min=<0 0>}
    }
   }
   slot {class=BorderSlot, properties {padding=<ltrb 0 0 0 0>}}
  }
  Background {widgetClassPath=&UserWidget, bInherited=true, widget=null}
 ]
}
