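# SettingEntryBackground is a UserWidget that gives settings entries a hover-highlight backdrop.
# A named Content slot sits in a minimum-height SizeBox, layered over a dark UIKit backplate that starts transparent.
# DisplaySelectedVisual either plays the short OnHover opacity fade forward or in reverse, or snaps opacity when Instant.
# Mouse enter and leave drive it except when touch is in use, and entering also plays BackgroundHoverSound.
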WidgetBlueprint /FortniteCoreUI/Settings/SettingEntries/SettingEntryBackground
animations [OnHover {
  binding_count=1
  length_seconds=0.1
  object=&self:OnHover
  sequence {
   bindings [WBP_UIKit_Backplate {
     id=#1
     possessed_class=WBP_UIKit_Backplate_C
     tracks [{
       type=MovieSceneFloatTrack
       display_name=Render Opacity
       sections [{type=MovieSceneFloatSection, end_frame=12, start_frame=0, channels [None {type=MovieSceneScriptingFloatChannel, keys [[0,0],[12,1]]}]}]
      }]
    }]
   display_rate {numerator=120}
  }
 }]
graphs [
 DisplaySelectedVisual {nodes [
   n1 = DisplaySelectedVisual!() -> (then, IsSelected:bool=~, Instant:bool=~) {is=title K2Node_FunctionEntry}
   > n2 = Branch(Condition=@n1.Instant)
   > StopAnimation(InAnimation=Get OnHover() -> (OnHover:Widget Animation Object Reference))
   > SetRenderOpacity(
    self=Get WBP_UIKit_Backplate() -> (WBP_UIKit_Backplate:WBP UIKit Backplate Object Reference)
    InOpacity=SelectFloat(A=1, B=0, bPickA=Get IsSelected() -> (IsSelected:bool))
   )
   n3 = Branch(execute=@n2.else, Condition=@n1)
   n4 = Get OnHover()
   > PlayAnimationForward(InAnimation=@n4)
   PlayAnimationReverse(execute=@n3.else, InAnimation=@n4)
  ]}
 EventGraph {nodes [
   Event OnMouseEnter()
   n1 = IsUsingTouch(PlayerController=GetOwningPlayer())
   > n2 = Branch(Condition=@n1)
   DisplaySelectedVisual!(execute=@n2.else, self, IsSelected:bool=true, Instant:bool=false) -> (then) {is=title K2Node_CallFunction}
   > SpawnSound2D!(
    execute
    Sound:Sound Base Object Reference=Get BackgroundHoverSound() -> (BackgroundHoverSound:Sound Base Object Reference)
    VolumeMultiplier:float=1
    PitchMultiplier:float=1
    StartTime:float=0
    ConcurrencySettings:Sound Concurrency Object Reference
    bPersistAcrossLevelTransition:bool=false
    bAutoDestroy:bool=true
   ) -> (then, ReturnValue:Audio Component Object Reference)
   Event OnMouseLeave!() -> (OutputDelegate, then, MouseEvent:"Pointer Event Structure (by ref)") {is=title K2Node_Event}
   > n3 = Branch(Condition=@n1)
   DisplaySelectedVisual(execute=@n3.else, IsSelected=false, Instant=false)
  ]}
]
variables [BackgroundHoverSound {type=SoundBase, default=null}]
widgets {
 info {inheritedWidgetCount=0, parentClass=&UserWidget, rootWidgetClass=&GridPanel, widgetCount=4}
 widgets [
  GridPanel_0 {widgetClassPath=&GridPanel, widget=GridPanel_0, properties {columnFill [1], rowFill [1]}}
  WBP_UIKit_Backplate {
   widgetClassPath=&WBP_UIKit_Backplate
   bIsVariable=true
   parent=GridPanel_0
   widget=WBP_UIKit_Backplate
   properties {brightness=Dark, cornerRadius=Interactive Component, renderOpacity=0}
   slot {class=GridSlot, properties {layer=-1}}
  }
  SizeBox_2 {
   widgetClassPath=&SizeBox
   parent=GridPanel_0
   widget=SizeBox_2
   properties {bOverride_MinDesiredHeight=true, heightOverride=64, minDesiredHeight=64}
   slot {class=GridSlot, properties {layer=1}}
  }
  Content {
   widgetClassPath=&NamedSlot
   bIsVariable=true
   parent=SizeBox_2
   widget=Content
   properties {bExposeOnInstanceOnly=true}
   slot {class=SizeBoxSlot}
  }
 ]
}
