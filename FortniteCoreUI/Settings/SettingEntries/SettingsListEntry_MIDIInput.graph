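# Settings list row widget for binding MIDI input to an action, with primary, secondary and clear buttons.
# The key buttons prompt rebinding of slot 0 or 1, Clear removes bindings, and the native parent supplies button text through OnUpdateKeyText.
# Mouse hover or touch selection plays the OnHover highlight; touch also updates the background's selected visual.
# Construct swaps the focus outline material on mobile, and releasing the entry resets the highlight.
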WidgetBlueprint /FortniteCoreUI/Settings/SettingEntries/SettingsListEntry_MIDIInput
animations [OnHover {
  binding_count=2
  length_seconds=0.1
  object=&self:OnHover
  sequence {bindings [
    Text_SettingName {
     id=#1
     possessed_class=CommonTextBlock
     tracks [{
       type=MovieSceneColorTrack
       display_name=Color and Opacity
       sections [{
         type=MovieSceneColorSection
         end_frame=2
         start_frame=0
         channels [
          Color.R {type=MovieSceneScriptingFloatChannel, keys [[0,0.521],[2,1]]}
          Color.G {type=MovieSceneScriptingFloatChannel, keys [[0,0.521],[2,1]]}
          Color.B {type=MovieSceneScriptingFloatChannel, keys [[0,0.521],[2,1]]}
          Color.A {type=MovieSceneScriptingFloatChannel, keys [[0,0.8],[2,1]]}
         ]
        }]
      }]
    }
    FocusOutline {
     id=#2
     possessed_class=Image
     tracks [{
       type=MovieSceneFloatTrack
       display_name=Render Opacity
       sections [{type=MovieSceneFloatSection, end_frame=2, start_frame=0, channels [None {type=MovieSceneScriptingFloatChannel, keys [[0,0],[2,1]]}]}]
      }]
    }
   ]}
 }]
graphs [
 Change Touch Selection State {nodes [
   n1 = Change Touch Selection State!() -> (then, IsSelected:bool=~) {is=title K2Node_FunctionEntry}
   > n2 = Branch(Condition=@n1)
   n3 = Get bIsInSelectedState() -> (bIsInSelectedState:bool)
   > n4 = Branch(Condition=@n3)
   Set bIsInSelectedState(execute=@n4.else, bIsInSelectedState:bool=true) -> (Output_Get:bool)
   n5 = Get OnHover() -> (OnHover:Widget Animation Object Reference)
   > PlayAnimationForward(InAnimation=@n5)
   n6 = GetSettingEntryBackground!(self) -> (SettingEntryBackground:Setting Entry Background Object Reference)
   > DisplaySelectedVisual!(execute, self:Setting Entry Background Object Reference=@n6, IsSelected:bool=true, Instant:bool=false) -> (then)
   > SpawnSound2D!(
    execute
    Sound:Sound Base Object Reference=Get HoverSound() -> (HoverSound:Sound Base Object Reference)
    VolumeMultiplier:float=1
    PitchMultiplier:float=1
    StartTime:float=0
    ConcurrencySettings:Sound Concurrency Object Reference
    bPersistAcrossLevelTransition:bool=false
    bAutoDestroy:bool=true
   ) -> (then, ReturnValue:Audio Component Object Reference)
   Branch(execute=@n2.else, Condition=@n3)
   > Set bIsInSelectedState(bIsInSelectedState=false)
   > PlayAnimationReverse(InAnimation=@n5, PlaybackSpeed=2)
   > DisplaySelectedVisual(self=@n6, IsSelected=false, Instant=false)
  ]}
 EventGraph {nodes [
   "On Button Clicked (Button_PrimaryKey)"!() -> (OutputDelegate, then, Button:Common Button Legacy Object Reference) {is=title K2Node_ComponentBoundEvent}
   > PromptChangeMIDIInputBinding!(execute, self:Fort Settings List Entry Setting MIDIInput Object Reference, MIDIInputSlot:int=0) -> (then)
   "On Button Clicked (Button_SecondaryKey)"!() -> (OutputDelegate, then, Button:Common Button Legacy Object Reference) {is=title K2Node_ComponentBoundEvent}
   > PromptChangeMIDIInputBinding(MIDIInputSlot=1)
   "On Button Clicked (Button_Clear)"!() -> (OutputDelegate, then, Button:Common Button Legacy Object Reference) {is=title K2Node_ComponentBoundEvent}
   > ClearMIDIInputBindings(self:Fort Settings List Entry Setting MIDIInput Object Reference)
   n1 = Event OnUpdateKeyText!() -> (OutputDelegate, then, PrimaryKeyText:"Text (by ref)", SecondaryKeyText:"Text (by ref)") {is=title K2Node_Event}
   > Set Text!(
    execute
    self:Scrolling Text Button Object Reference=Get Button_PrimaryKey() -> (Button_PrimaryKey:Scrolling Text Button Object Reference)
    ButtonText:Text=@n1.PrimaryKeyText
   ) -> (then) {is=title K2Node_CallFunction}
   > Set Text(self=Get Button_SecondaryKey() -> (Button_SecondaryKey:Scrolling Text Button Object Reference), ButtonText=@n1.SecondaryKeyText)
   n2 = Event BP_UpdateTouchSelectionState!() -> (OutputDelegate, then, bIsSelected:bool) {is=title K2Node_Event}
   > Change Touch Selection State!(execute, self, IsSelected:bool=@n2.bIsSelected) -> (then) {is=title K2Node_CallFunction}
   Event On Entry Released() {is=title K2Node_Event}
   > Set bIsInSelectedState(bIsInSelectedState=false)
   n3 = Get OnHover()
   > PlayAnimation(InAnimation=@n3, StartAtTime=GetEndTime(self=@n3), PlayMode=Reverse)
   > DisplaySelectedVisual(self=GetSettingEntryBackground(), IsSelected=false, Instant=true)
   Event OnMouseEnter()
   n4 = IsUsingTouch(PlayerController=GetOwningPlayer())
   > n5 = Branch(Condition=@n4)
   n6 = Get OnHover()
   PlayAnimationForward(execute=@n5.else, InAnimation=@n6)
   > Set bIsInSelectedState(bIsInSelectedState=true)
   > SpawnSound2D(Sound=Get HoverSound(), VolumeMultiplier=1, PitchMultiplier=1, StartTime=0, bPersistAcrossLevelTransition=false, bAutoDestroy=true)
   Event OnMouseLeave!() -> (OutputDelegate, then, MouseEvent:"Pointer Event Structure (by ref)") {is=title K2Node_Event}
   > n7 = Branch(Condition=@n4)
   PlayAnimationReverse(execute=@n7.else, InAnimation=@n6)
   > Set bIsInSelectedState(bIsInSelectedState=false)
   Event Construct()
   > Branch(Condition=IsMobileGame())
   > SetBrush(
    self=Get FocusOutline() -> (FocusOutline:Image Object Reference)
    InBrush=MakeBrushFromMaterial!(Material:Material Interface Object Reference=/FortniteCoreUI/Settings/Art/MI_UIKit_Settings_Outline_Bacchus, Width:int=32, Height:int=32) -> (ReturnValue:Slate Brush Structure)
   )
  ]}
 GetPrimaryGamepadFocusWidget {nodes [
   GetPrimaryGamepadFocusWidget() {is=title K2Node_FunctionEntry}
   > Return Node!(execute, ReturnValue:Widget Object Reference=Get Button_PrimaryKey())
  ]}
 GetSettingEntryBackground {nodes [
   GetSettingEntryBackground() {is=title K2Node_FunctionEntry}
   > n1 = Cast To SettingEntryBackground!(execute, Object=Get Background() -> (Background:User Widget Object Reference)) -> (then, CastFailed, AsSetting Entry Background:Setting Entry Background Object Reference) {is=title K2Node_DynamicCast}
   > Return Node!(execute, SettingEntryBackground:Setting Entry Background Object Reference=@n1)
  ]}
]
variables [
 bIsInSelectedState {type=boolean}
 HoverSound {type=SoundBase, default=/Game/Sounds/Fort_UI_Sounds/FrontEnd/DefaultSounds/PressedButtons/MSS_UI_Button_Hover_Default}
]
widgets {
 info {
  inheritedWidgetCount=2
  namedSlotCount=1
  parentClass=/Script/FortniteUI.FortSettingsListEntrySetting_MIDIInput
  rootWidgetClass=&/SettingEntryBackground.SettingEntryBackground_C
  widgetCount=11
 }
 widgets [
  Background {widgetClassPath=&/SettingEntryBackground, bInherited=true, bIsVariable=true, widget=Background}
  Overlay_2 {widgetClassPath=&Overlay, namedSlotHost=Background, widget=Overlay_2}
  HorizontalBox_1 {
   widgetClassPath=&HorizontalBox
   parent=Overlay_2
   widget=HorizontalBox_1
   slot {class=OverlaySlot, properties {horizontalAlignment=HAlign_Fill, verticalAlignment=VAlign_Center}}
  }
  Text_SettingName {
   widgetClassPath=&CommonTextBlock
   bInherited=true
   bIsVariable=true
   parent=HorizontalBox_1
   widget=Text_SettingName
   properties {
    autoWrapText=true
    scrollStyle=&TextScrollStyle-Base_C
    text="{Setting Name}"
    wrapTextAt=900
    colorAndOpacity {colorUseRule=UseColor_Specified, specifiedColor=<rgba 0.5216 0.5216 0.5216 0.8>}
    font {
     bForceMonospaced=false
     bMaterialIsStencil=false
     fontMaterial=null
     fontObject=&HeadingNow
     letterSpacing=0
     monospacedWidth=1
     size=18
     skewAmount=0
     typefaceFontName=75 Medium
     outlineSettings {
      bApplyOutlineToDropShadows=false
      bMiteredCorners=false
      bSeparateFillAlpha=false
      outlineMaterial=null
      outlineSize=0
      outlineColor=<rgba 0 0 0 1>
     }
    }
    shadowOffset=<0 0>
   }
   slot {class=HorizontalBoxSlot, properties {verticalAlignment=VAlign_Center, padding=<ltrb 24 0 10 0>, size {sizeRule=Fill, value=1}}}
  }
  HorizontalBox_2 {
   widgetClassPath=&HorizontalBox
   parent=HorizontalBox_1
   widget=HorizontalBox_2
   slot {class=HorizontalBoxSlot, properties {verticalAlignment=VAlign_Center, size {sizeRule=Fill, value=0.8}}}
  }
  Overlay_0 {
   widgetClassPath=&Overlay
   parent=HorizontalBox_2
   widget=Overlay_0
   slot {class=HorizontalBoxSlot, properties {padding=<ltrb 8 0 8 0>, size {sizeRule=Fill, value=1}}}
  }
  Button_PrimaryKey {
   widgetClassPath=/FortUILibrary/Foundation/Buttons/ScrollingTextButton
   bIsVariable=true
   parent=Overlay_0
   widget=Button_PrimaryKey
   properties {
    bHideInputAction=true
    onButtonBaseClicked=(&self:WidgetTree.Button_PrimaryKey.HandleOnButtonClicked)
    onButtonBaseDoubleClicked=(&self:WidgetTree.Button_PrimaryKey.HandleOnButtonDoubleClicked)
    onButtonBaseHovered=(&self:WidgetTree.Button_PrimaryKey.HandleOnButtonHovered)
    onButtonBaseUnhovered=(&self:WidgetTree.Button_PrimaryKey.HandleOnButtonUnhovered)
    onSelectedChangedBase=(&self:WidgetTree.Button_PrimaryKey.HandleOnSelectedChanged)
    style=/FortniteCoreUI/Settings/SettingEntries/Styles/Settings_ButtonStyle_PurpleButton_Settings.Settings_ButtonStyle_PurpleButton_Settings_C
   }
   slot {class=OverlaySlot, properties {horizontalAlignment=HAlign_Fill, verticalAlignment=VAlign_Fill}}
  }
  Overlay_SecondaryKey {
   widgetClassPath=&Overlay
   parent=HorizontalBox_2
   widget=Overlay_SecondaryKey
   slot {class=HorizontalBoxSlot, properties {padding=<ltrb 8 0 8 0>, size {sizeRule=Fill, value=1}}}
  }
  Button_SecondaryKey {
   widgetClassPath=&ScrollingTextButton
   bIsVariable=true
   parent=Overlay_SecondaryKey
   widget=Button_SecondaryKey
   properties {
    bHideInputAction=true
    navigation=&self:WidgetTree.Button_SecondaryKey.WidgetNavigation_0
    onButtonBaseClicked=(&self:WidgetTree.Button_SecondaryKey.HandleOnButtonClicked)
    onButtonBaseDoubleClicked=(&self:WidgetTree.Button_SecondaryKey.HandleOnButtonDoubleClicked)
    onButtonBaseHovered=(&self:WidgetTree.Button_SecondaryKey.HandleOnButtonHovered)
    onButtonBaseUnhovered=(&self:WidgetTree.Button_SecondaryKey.HandleOnButtonUnhovered)
    onSelectedChangedBase=(&self:WidgetTree.Button_SecondaryKey.HandleOnSelectedChanged)
    style=&Settings_ButtonStyle_PurpleButton_Settings_C
   }
   slot {class=OverlaySlot, properties {horizontalAlignment=HAlign_Fill, verticalAlignment=VAlign_Fill}}
  }
  Button_Clear {
   widgetClassPath=&IconTextButton
   bIsVariable=true
   parent=HorizontalBox_2
   widget=Button_Clear
   properties {
    bHideInputAction=true
    buttonText=""
    contentAlignment=HAlign_Center
    navigation=&self:WidgetTree.Button_Clear.WidgetNavigation_0
    onButtonBaseClicked=(&self:WidgetTree.Button_Clear.HandleOnButtonClicked)
    onButtonBaseDoubleClicked=(&self:WidgetTree.Button_Clear.HandleOnButtonDoubleClicked)
    onButtonBaseHovered=(&self:WidgetTree.Button_Clear.HandleOnButtonHovered)
    onButtonBaseUnhovered=(&self:WidgetTree.Button_Clear.HandleOnButtonUnhovered)
    onSelectedChangedBase=(&self:WidgetTree.Button_Clear.HandleOnSelectedChanged)
    style=&Settings_ButtonStyle_PurpleButton_Settings_C
    font {
     bForceMonospaced=false
     bMaterialIsStencil=false
     fontMaterial=null
     fontObject=&HeadingNow
     letterSpacing=0
     monospacedWidth=1
     size=14.25
     skewAmount=0
     typefaceFontName=75 Medium
     outlineSettings {
      bApplyOutlineToDropShadows=false
      bMiteredCorners=false
      bSeparateFillAlpha=false
      outlineMaterial=null
      outlineSize=0
      outlineColor=<rgba 0 0 0 1>
     }
    }
    hoveredSlateSoundOverride {resourceObject=null}
    iconBrush {
     bIsDynamicallyLoaded=false
     drawAs=Image
     imageType=NoImage
     mirroring=NoMirror
     resourceName=None
     resourceObject=/Game/UI/Foundation/Textures/Buttons/T-SlotIcon-Remove-32
     tiling=NoTile
     imageSize=<32 32>
     margin=<ltrb 0 0 0 0>
     outlineSettings {
      bUseBrushTransparency=false
      roundingType=HalfHeightRadius
      width=0
      color {colorUseRule=UseColor_Specified, specifiedColor=<rgba 0 0 0 0>}
      cornerRadii=<0 0 0 1>
     }
     tintColor {colorUseRule=UseColor_Specified, specifiedColor=<rgba 0.6 0.036 0.036 1>}
     uVRegion {bIsValid=false, max=<0 0>, min=<0 0>}
    }
    pressedSlateSoundOverride {resourceObject=null}
   }
   slot {class=HorizontalBoxSlot, properties {padding=<ltrb 8 0 8 0>}}
  }
  FocusOutline {
   widgetClassPath=&Image
   bIsVariable=true
   parent=Overlay_2
   widget=FocusOutline
   properties {
    pixelSnapping=Disabled
    renderOpacity=0
    visibility=SelfHitTestInvisible
    brush {
     bIsDynamicallyLoaded=false
     drawAs=Image
     imageType=NoImage
     mirroring=NoMirror
     resourceName=None
     resourceObject=/FortniteCoreUI/Settings/Art/MI_UIKit_Settings_Outline
     tiling=NoTile
     imageSize=<32 32>
     margin=<ltrb 0 0 0 0>
     outlineSettings {
      bUseBrushTransparency=false
      roundingType=HalfHeightRadius
      width=0
      color {colorUseRule=UseColor_Specified, specifiedColor=<rgba 0 0 0 0>}
      cornerRadii=<0 0 0 0>
     }
     tintColor {colorUseRule=UseColor_Specified, specifiedColor=<rgba 1 1 1 1>}
     uVRegion {bIsValid=false, max=<0 0>, min=<0 0>}
    }
   }
   slot {class=OverlaySlot, properties {horizontalAlignment=HAlign_Fill, verticalAlignment=VAlign_Fill, padding=<ltrb -16 -16 -16 -16>}}
  }
 ]
}
