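# Settings list row widget for a scalar setting, built on FortSettingsListEntrySetting_Scalar, showing a name and a slider value.
# A transparent AnalogSlider takes input and gamepad focus over a material progress bar driven by OnValueChanged.
# Mouse hover and touch selection play a focus animation chosen by editable state, with touch also driving the background's selected visual.
# Capture and selection play sounds, non-editable rows dim, and compact UI swaps the outline material.
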WidgetBlueprint /FortniteCoreUI/Settings/SettingEntries/SettingsListEntry_Scalar
animations [
 OnHover {
  binding_count=3
  length_seconds=0.1
  object=&self:OnHover
  sequence {
   bindings [
    Text_SettingValue {
     id=#1
     possessed_class=CommonTextBlock
     tracks [{
       type=MovieSceneColorTrack
       display_name=Color and Opacity
       sections [{
         type=MovieSceneColorSection
         end_frame=12
         start_frame=0
         channels [
          Color.R {type=MovieSceneScriptingFloatChannel, keys [[0,0.521],[12,1]]}
          Color.G {type=MovieSceneScriptingFloatChannel, keys [[0,0.521],[12,1]]}
          Color.B {type=MovieSceneScriptingFloatChannel, keys [[0,0.521],[12,1]]}
          Color.A {type=MovieSceneScriptingFloatChannel, keys [[0,0.8],[12,1]]}
         ]
        }]
      }]
    }
    FocusOutline {
     id=#2
     possessed_class=Image
     tracks [{
       type=MovieSceneFloatTrack
       display_name=Render Opacity
       sections [{type=MovieSceneFloatSection, end_frame=12, start_frame=0, channels [None {type=MovieSceneScriptingFloatChannel, keys [[0,0],[12,1]]}]}]
      }]
    }
    Text_SettingName {
     id=#3
     possessed_class=CommonTextBlock
     tracks [{
       type=MovieSceneColorTrack
       display_name=Color and Opacity
       sections [{
         type=MovieSceneColorSection
         end_frame=12
         start_frame=0
         channels [
          Color.R {type=MovieSceneScriptingFloatChannel, keys [[0,0.521],[12,1]]}
          Color.G {type=MovieSceneScriptingFloatChannel, keys [[0,0.521],[12,1]]}
          Color.B {type=MovieSceneScriptingFloatChannel, keys [[0,0.521],[12,1]]}
          Color.A {type=MovieSceneScriptingFloatChannel, keys [[0,0.8],[12,1]]}
         ]
        }]
      }]
    }
   ]
   display_rate {numerator=120}
  }
 }
 OnSelect_Touch {
  binding_count=1
  length_seconds=0.5
  object=&self:OnSelect_Touch
  sequence {bindings [Background {
     id=#4
     possessed_class=SettingEntryBackground_C
     tracks [{
       type=MovieSceneColorTrack
       display_name=Color and Opacity
       sections [{
         type=MovieSceneColorSection
         end_frame=10
         start_frame=0
         channels [
          Color.R {type=MovieSceneScriptingFloatChannel, const [1,1]}
          Color.G {type=MovieSceneScriptingFloatChannel, keys [[0,1],[10,0]]}
          Color.B {type=MovieSceneScriptingFloatChannel, keys [[0,1],[10,0]]}
          Color.A {type=MovieSceneScriptingFloatChannel, const [1,1]}
         ]
        }]
      }]
    }]}
 }
 OnHover_Disabled {
  binding_count=1
  length_seconds=0.1
  object=&self:OnHover_Disabled
  sequence {
   bindings [FocusOutline {
     id=#2
     possessed_class=Image
     tracks [{
       type=MovieSceneFloatTrack
       display_name=Render Opacity
       sections [{type=MovieSceneFloatSection, end_frame=12, start_frame=0, channels [None {type=MovieSceneScriptingFloatChannel, keys [[0,0],[12,1]]}]}]
      }]
    }]
   display_rate {numerator=120}
  }
 }
]
graphs [
 Change Touch Selection State {nodes [
   n1 = Change Touch Selection State!() -> (then, IsSelected:bool=~) {is=title K2Node_FunctionEntry}
   > n2 = Branch(Condition=@n1)
   n3 = Get bIsInSelectedState() -> (bIsInSelectedState:bool)
   > n4 = Branch(Condition=@n3)
   Set bIsInSelectedState(execute=@n4.else, bIsInSelectedState:bool=true) -> (Output_Get:bool)
   n5 = GetOnHoverAnimation!() -> (OnHoverAnimation:Widget Animation Object Reference)
   > PlayAnimationForward(InAnimation=@n5)
   n6 = GetSettingEntryBackground!(self) -> (SettingEntryBackground:Setting Entry Background Object Reference)
   > DisplaySelectedVisual!(execute, self:Setting Entry Background Object Reference=@n6, IsSelected:bool=true, Instant:bool=false) -> (then)
   > SpawnSound2D!(
    execute
    Sound:Sound Base Object Reference=Get HoverSound() -> (HoverSound:Sound Base Object Reference)
    VolumeMultiplier:float=1
    PitchMultiplier:float=1
    StartTime:float=0
    ConcurrencySettings:Sound Concurrency Object Reference
    bPersistAcrossLevelTransition:bool=false
    bAutoDestroy:bool=true
   ) -> (then, ReturnValue:Audio Component Object Reference)
   Branch(execute=@n2.else, Condition=@n3)
   > Set bIsInSelectedState(bIsInSelectedState=false)
   > PlayAnimationReverse(InAnimation=@n5, PlaybackSpeed=2)
   > DisplaySelectedVisual(self=@n6, IsSelected=false, Instant=false)
  ]}
 EventGraph {nodes [
   n1 = "On Mouse Capture Begin (Slider_SettingValue)"() {is=title K2Node_ComponentBoundEvent}
   SpawnSound2D(
    execute=[@n1.then, @n2.then]
    Sound=Get PressedSound() -> (PressedSound:Sound Base Object Reference)
    VolumeMultiplier=1
    PitchMultiplier=1
    StartTime=0
    bPersistAcrossLevelTransition=false
    bAutoDestroy=true
   )
   n2 = "On Controller Capture Begin (Slider_SettingValue)"() {is=title K2Node_ComponentBoundEvent}
   n3 = Event OnValueChanged!() -> (OutputDelegate, then, Value:float) {is=title K2Node_Event}
   > SetProgress!(
    execute
    self:Fort Simple Material Progress Bar Object Reference=Get Progress() -> (Progress:Simple Material Progress Bar Object Reference)
    Progress:float=@n3.Value
   ) -> (then)
   n4 = Event BP_UpdateTouchSelectionState!() -> (OutputDelegate, then, bIsSelected:bool) {is=title K2Node_Event}
   > Change Touch Selection State!(execute, self, IsSelected:bool=@n4.bIsSelected) -> (then) {is=title K2Node_CallFunction}
   n5 = Event OnEditableStateChanged!() -> (OutputDelegate, then, bIsEditable:bool) {is=title K2Node_Event}
   > SetSettingTextEditableState!(execute, self, IsEditable:bool=@n5.bIsEditable) -> (then)
   Event On Entry Released() {is=title K2Node_Event}
   > Set bIsInSelectedState(bIsInSelectedState=false)
   n6 = GetOnHoverAnimation()
   > PlayAnimation(InAnimation=@n6, StartAtTime=GetEndTime(self=@n6), PlayMode=Reverse)
   > DisplaySelectedVisual(self=GetSettingEntryBackground(), IsSelected=false, Instant=true)
   Event OnMouseEnter()
   n7 = IsUsingTouch(PlayerController=GetOwningPlayer())
   > n8 = Branch(Condition=@n7)
   n9 = GetOnHoverAnimation()
   PlayAnimationForward(execute=@n8.else, InAnimation=@n9)
   > Set bIsInSelectedState(bIsInSelectedState=true)
   > SpawnSound2D(Sound=Get HoverSound(), VolumeMultiplier=1, PitchMultiplier=1, StartTime=0, bPersistAcrossLevelTransition=false, bAutoDestroy=true)
   Event OnMouseLeave!() -> (OutputDelegate, then, MouseEvent:"Pointer Event Structure (by ref)") {is=title K2Node_Event}
   > n10 = Branch(Condition=@n7)
   PlayAnimationReverse(execute=@n10.else, InAnimation=@n9)
   > Set bIsInSelectedState(bIsInSelectedState=false)
   Event Construct()
   > Branch(Condition=IsUsingCompactUI())
   > SetBrush(
    self=Get FocusOutline() -> (FocusOutline:Image Object Reference)
    InBrush=MakeBrushFromMaterial!(Material:Material Interface Object Reference=/FortniteCoreUI/Settings/Art/MI_UIKit_Settings_Outline_Bacchus, Width:int=32, Height:int=32) -> (ReturnValue:Slate Brush Structure)
   )
  ]}
 GetOnHoverAnimation {nodes [
   GetOnHoverAnimation() {is=title K2Node_FunctionEntry}
   > Return Node!(
    execute
    OnHoverAnimation:Widget Animation Object Reference=Select(
     Option 0:Widget Animation Object Reference=Get OnHover_Disabled() -> (OnHover_Disabled:Widget Animation Object Reference)
     Option 1:Widget Animation Object Reference=Get OnHover() -> (OnHover:Widget Animation Object Reference)
     Index=IsEditable!() -> (IsEditable:bool)
    ) -> (ReturnValue:Widget Animation Object Reference)
   )
  ]}
 GetPrimaryGamepadFocusWidget {nodes [
   GetPrimaryGamepadFocusWidget() {is=title K2Node_FunctionEntry}
   > Return Node!(execute, ReturnValue:Widget Object Reference=Get Slider_SettingValue() -> (Slider_SettingValue:Analog Slider Object Reference))
  ]}
 GetSettingEntryBackground {nodes [
   GetSettingEntryBackground() {is=title K2Node_FunctionEntry}
   > n1 = Cast To SettingEntryBackground!(execute, Object=Get Background() -> (Background:User Widget Object Reference)) -> (then, CastFailed, AsSetting Entry Background:Setting Entry Background Object Reference) {is=title K2Node_DynamicCast}
   > Return Node!(execute, SettingEntryBackground:Setting Entry Background Object Reference=@n1)
  ]}
 IsEditable {nodes [
   IsEditable() {is=title K2Node_FunctionEntry}
   > Return Node!(
    execute
    IsEditable:bool=Get bIsEnabled!(self:Widget Object Reference=Get Panel_Value() -> (Panel_Value:Panel Widget Object Reference)) -> (bIsEnabled:bool)
   )
  ]}
 SetSettingTextEditableState {nodes [
   n1 = SetSettingTextEditableState!() -> (then, IsEditable:bool=~) {is=title K2Node_FunctionEntry}
   > SetColorAndOpacity(
    self=Get Text_SettingName() -> (Text_SettingName:Common Text Object Reference)
    InColorAndOpacity=Select(
     Option 0:Slate Color Structure=Make Slate Color(SpecifiedColor=(R=0.2083,G=0.2083,B=0.2083,A=0.8))
     Option 1:Slate Color Structure=Make Slate Color(SpecifiedColor=(R=0.5216,G=0.5216,B=0.5216,A=0.8))
     Index=@n1
    ) -> (ReturnValue:Slate Color Structure)
   )
   > SetRenderOpacity(self=Get Panel_Value(), InOpacity=SelectFloat(A=1, B=0.25, bPickA=@n1))
  ]}
]
variables [
 bIsInSelectedState {type=boolean}
 HoverSound {type=SoundBase, default=/Game/Sounds/Fort_UI_Sounds/FrontEnd/DefaultSounds/PressedButtons/MSS_UI_Button_Hover_Default}
 PressedSound {type=SoundBase, default=/Game/Sounds/Fort_UI_Sounds/FrontEnd/DefaultSounds/PressedButtons/MSS_UI_Button_Pressed_Default}
]
widgets {
 info {
  inheritedWidgetCount=5
  namedSlotCount=1
  parentClass=/Script/FortniteUI.FortSettingsListEntrySetting_Scalar
  rootWidgetClass=&/SettingEntryBackground.SettingEntryBackground_C
  widgetCount=11
 }
 widgets [
  Background {widgetClassPath=&/SettingEntryBackground, bInherited=true, bIsVariable=true, widget=Background}
  Overlay_0 {widgetClassPath=&Overlay, namedSlotHost=Background, widget=Overlay_0}
  FortResponsivePanel_1 {
   widgetClassPath=/Script/FortniteUI.FortResponsivePanel
   parent=Overlay_0
   widget=FortResponsivePanel_1
   properties {bCanStackVertically=false}
   slot {class=OverlaySlot, properties {horizontalAlignment=HAlign_Fill}}
  }
  SizeBox_0 {
   widgetClassPath=&SizeBox
   parent=FortResponsivePanel_1
   widget=SizeBox_0
   properties {bOverride_MinDesiredHeight=true, minDesiredHeight=64}
   slot {class=FortResponsivePanelSlot}
  }
  Text_SettingName {
   widgetClassPath=&CommonTextBlock
   bInherited=true
   bIsVariable=true
   parent=SizeBox_0
   widget=Text_SettingName
   properties {
    scrollStyle=&TextScrollStyle-Base_C
    text="{Setting Name}"
    colorAndOpacity {colorUseRule=UseColor_Specified, specifiedColor=<rgba 0.521 0.521 0.521 0.8>}
    font {
     bForceMonospaced=false
     bMaterialIsStencil=false
     fontMaterial=null
     fontObject=&HeadingNow
     letterSpacing=0
     monospacedWidth=1
     size=18
     skewAmount=0
     typefaceFontName=75 Medium
     outlineSettings {
      bApplyOutlineToDropShadows=false
      bMiteredCorners=false
      bSeparateFillAlpha=false
      outlineMaterial=null
      outlineSize=0
      outlineColor=<rgba 0 0 0 1>
     }
    }
    shadowOffset=<0 0>
   }
   slot {class=SizeBoxSlot, properties {horizontalAlignment=HAlign_Left, verticalAlignment=VAlign_Center, padding=<ltrb 24 0 10 0>}}
  }
  SizeBox_1 {
   widgetClassPath=&SizeBox
   parent=FortResponsivePanel_1
   widget=SizeBox_1
   properties {bOverride_MinDesiredHeight=true, minDesiredHeight=50}
   slot {class=FortResponsivePanelSlot}
  }
  Panel_Value {widgetClassPath=&Overlay, bInherited=true, bIsVariable=true, parent=SizeBox_1, widget=Panel_Value, slot {class=SizeBoxSlot}}
  Progress {
   widgetClassPath=/Game/Athena/HUD/SimpleMaterialProgressBar
   bIsVariable=true
   parent=Panel_Value
   widget=Progress
   properties {
    bOverrideDefaultColorA=true
    bOverrideDefaultColorB=true
    bOverrideDefaultColorBackground=true
    visibility=HitTestInvisible
    cachedColorA=<rgba 0 0.0232 0.2502 1>
    cachedColorB=<rgba 0 0.0232 0.2502 1>
    cachedColorBackground=<rgba 0 0 0.0331 1>
   }
   slot {class=OverlaySlot, properties {horizontalAlignment=HAlign_Fill, verticalAlignment=VAlign_Fill}}
  }
  Text_SettingValue {
   widgetClassPath=&CommonTextBlock
   bInherited=true
   bIsVariable=true
   parent=Panel_Value
   widget=Text_SettingValue
   properties {
    justification=Center
    scrollStyle=&TextScrollStyle-Base_C
    text="{Value}"
    visibility=HitTestInvisible
    colorAndOpacity {colorUseRule=UseColor_Specified, specifiedColor=<rgba 0.521 0.521 0.521 0.8>}
    font {
     bForceMonospaced=false
     bMaterialIsStencil=false
     fontMaterial=null
     fontObject=&HeadingNow
     letterSpacing=0
     monospacedWidth=1
     size=18
     skewAmount=0
     typefaceFontName=74 Regular
     outlineSettings {
      bApplyOutlineToDropShadows=false
      bMiteredCorners=false
      bSeparateFillAlpha=false
      outlineMaterial=null
      outlineSize=0
      outlineColor=<rgba 0 0 0 1>
     }
    }
    shadowOffset=<0 0>
   }
   slot {class=OverlaySlot, properties {horizontalAlignment=HAlign_Fill, verticalAlignment=VAlign_Center}}
  }
  Slider_SettingValue {
   widgetClassPath=/Script/CommonUI.AnalogSlider
   bInherited=true
   bIsVariable=true
   parent=Panel_Value
   widget=Slider_SettingValue
   properties {sliderBarColor=<rgba 1 1 1 0>, sliderHandleColor=<rgba 1 1 1 0>}
   slot {class=OverlaySlot, properties {horizontalAlignment=HAlign_Fill, verticalAlignment=VAlign_Fill}}
  }
  FocusOutline {
   widgetClassPath=&Image
   bIsVariable=true
   parent=Overlay_0
   widget=FocusOutline
   properties {
    pixelSnapping=Disabled
    renderOpacity=0
    visibility=SelfHitTestInvisible
    brush {
     bIsDynamicallyLoaded=false
     drawAs=Image
     imageType=NoImage
     mirroring=NoMirror
     resourceName=None
     resourceObject=/FortniteCoreUI/Settings/Art/MI_UIKit_Settings_Outline
     tiling=NoTile
     imageSize=<32 32>
     margin=<ltrb 0 0 0 0>
     outlineSettings {
      bUseBrushTransparency=false
      roundingType=HalfHeightRadius
      width=0
      color {colorUseRule=UseColor_Specified, specifiedColor=<rgba 0 0 0 0>}
      cornerRadii=<0 0 0 0>
     }
     tintColor {colorUseRule=UseColor_Specified, specifiedColor=<rgba 1 1 1 1>}
     uVRegion {bIsValid=false, max=<0 0>, min=<0 0>}
    }
   }
   slot {class=OverlaySlot, properties {horizontalAlignment=HAlign_Fill, verticalAlignment=VAlign_Fill, padding=<ltrb -16 -16 -16 -16>}}
  }
 ]
}
